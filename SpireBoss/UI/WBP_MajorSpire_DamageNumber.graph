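# Major Spire damage number widget, a child of WBP_Spires_DamageNumbers that adds no widgets and picks number colors on each damage-dealt event.
# If the hit actor's parent has the Shielded tag or damage is zero, it calls the negative color setter, passing only the zero-damage result as IsNumberNegative.
# Otherwise, casts choose the colors: large weakpoints follow bForcedActive, core weakpoints always count as active, and other actors get default colors.
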WidgetBlueprint /SpireBoss/UI/WBP_MajorSpire_DamageNumber
graphs [EventGraph {nodes [
   Event OnInitialized()
   > Parent: OnInitialized() {is=title K2Node_CallParentFunction}
   n1 = Event ReceiveOnDamageDealt!() -> (
    OutputDelegate
    then
    InDamage:double
    InActor:Actor Object Reference
    bInShield:bool
    bInCrit:bool
    InWorldSpacePos:Vector
    bInEMP:bool
    DamageTags:"Gameplay Tag Container Structure (by ref)"
   ) {is=title K2Node_Event}
   > n2 = HasMatchingGameplayTag!(
    execute:Exec
    self:Object Reference=GetParentActor!(self:Actor Object Reference=@n1.InActor) -> (ReturnValue:Actor Object Reference)
    TagToCheck:Gameplay Tag Structure=Get MajorSpireShieldedTag() -> (MajorSpireShieldedTag:Gameplay Tag Structure)
   ) -> (then:Exec, ReturnValue:bool) {is=title K2Node_Message}
   n3 = "Nearly Equal (Float)"(A=@n1.InDamage, B=0, ErrorTolerance=0)
   > n4 = Branch(Condition=OR Boolean(A=@n2, B=@n3))
   > SetNegativeNumberColors!(execute, self:WBP Spires Damage Numbers Object Reference, IsNumberNegative:bool=@n3) -> (then)
   > Set IsWeakpointActive(IsWeakpointActive:bool=false) -> (Output_Get:bool)
   > Set IsDefaultNumber(IsDefaultNumber:bool=false) -> (Output_Get:bool)
   Set IsNegativeNumber(execute=@n4.else, IsNegativeNumber:bool=false) -> (Output_Get:bool)
   > n5 = Cast To BGA_WeakpointActor_Large_Spire!(execute, Object=@n1.InActor) -> (then, CastFailed, AsBGA Weakpoint Actor Large Spire:BGA Weakpoint Actor Large Spire Object Reference) {is=title K2Node_DynamicCast}
   > n6 = SetWeakpointNumberColors!(
    execute
    self:WBP Spires Damage Numbers Object Reference
    IsWeakpointActive:bool=Get bForcedActive!(self:BGA Weakpoint Actor Large Spire Object Reference=@n5) -> (bForcedActive:bool)
   ) -> (then)
   Set IsDefaultNumber(execute=[@n8.then, @n6.then], IsDefaultNumber=false)
   n7 = Cast To BGA_WeakpointActor_Core_Large_Spire!(execute=@n5.CastFailed, Object=@n1.InActor) -> (then, CastFailed, AsBGA Weakpoint Actor Core Large Spire:BGA Weakpoint Actor Core Large Spire Object Reference) {is=title K2Node_DynamicCast}
   > n8 = SetWeakpointNumberColors(IsWeakpointActive=true)
   SetDefaultNumberColors!(execute=@n7.CastFailed, self:WBP Spires Damage Numbers Object Reference, IsDefaultNumber:bool=true) -> (then)
  ]}]
variables [MajorSpireShieldedTag {type=GameplayTag, default {tagName=Spire.Major.State.Shielded}}]
widgets {
 info {
  inheritedWidgetCount=0
  parentClass=/SpireSharedAssets/UI/WBP_Spires_DamageNumbers.WBP_Spires_DamageNumbers_C
  rootWidgetClass=null
  widgetCount=0
 }
 widgets []
}
