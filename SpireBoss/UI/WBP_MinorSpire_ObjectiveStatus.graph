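# World-space label widget that shows a Minor Spire's health bar and an optional objective counter.
# SetHealthPercent sets the fill bar and animates a delta bar, which starts collapsed.
# Vulnerability toggles a lock icon and fill colours, while distance-threshold updates show the widget only at threshold zero.
# Valid objective data reveals the counter row with its maximum and an icon select fixed at index zero; a separate function updates progress text.
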WidgetBlueprint /SpireBoss/UI/WBP_MinorSpire_ObjectiveStatus
graphs [
 EventGraph {nodes [
   n1 = Event OnDistanceUpdate!() -> (OutputDelegate, then, PrevDistanceThreshold:int=0, NextDistanceThreshold:int=0) {is=title K2Node_Event}
   > Parent: OnDistanceUpdate!(execute, PrevDistanceThreshold:int=@n1.PrevDistanceThreshold, NextDistanceThreshold:int=@n1.NextDistanceThreshold) -> (then) {is=title K2Node_CallParentFunction}
   > UpdateHealthStatusVisibility!(execute, self, CurrentDistanceThreshold:int=@n1.NextDistanceThreshold) -> (then)
   Event Construct()
   > Parent: Construct()
   > SetHealthPercent!(execute, self, Progress:float=Get CurrentProgress() -> (CurrentProgress:float)) -> (then)
   > UpdateHealthStatusVisibility(CurrentDistanceThreshold=Get CurrentDistanceThreshold() -> (CurrentDistanceThreshold:int=0))
  ]}
 SetHealthPercent {nodes [
   n1 = SetHealthPercent!() -> (then, Progress:float=~) {is=title K2Node_FunctionEntry}
   > n2 = Set CurrentProgress(CurrentProgress:float=@n1) -> (Output_Get:float)
   > SetProgress!(
    execute
    self:Fort Simple Material Progress Bar Object Reference=Get ProgressBar_HealthFill() -> (ProgressBar_HealthFill:Simple Material Progress Bar Object Reference)
    Progress:float=@n2
   ) -> (then)
   > AnimateProgressFromCurrent!(
    execute
    self:Fort Simple Material Progress Bar Object Reference=Get ProgressBar_HealthDelta() -> (ProgressBar_HealthDelta:Simple Material Progress Bar Object Reference)
    End:float=@n2
    AnimSpeed:float=1
   ) -> (then)
  ]}
 SetObjectiveData {nodes [
   n1 = SetObjectiveData!() -> (then, InObjective:bool=~, Max Count:int) {is=title K2Node_FunctionEntry}
   > n2 = Set bIsObjectiveValid(bIsObjectiveValid:bool=AND Boolean(A=@n1, B="integer > integer"(A=@n1.Max Count))) -> (Output_Get:bool)
   > n3 = Branch(Condition=@n2)
   > SetVisibility(self=Get HBox_ObjectiveProgress() -> (HBox_ObjectiveProgress:Horizontal Box Object Reference), InVisibility=SelfHitTestInvisible)
   > "SetText (Text)"(
    self=Get Text_Obj_Max() -> (Text_Obj_Max:Common Text Object Reference)
    InText="To Text (Integer)"(Value=Get Max Count() -> (Max Count:int=0))
   )
   > SetBrushFromTexture!(
    execute
    self:Image Object Reference=Get Image_ObjectiveIcon() -> (Image_ObjectiveIcon:Image Object Reference)
    Texture:Texture 2D Object Reference=Select!(
     Option 0:Texture 2D Object Reference=Get SpireSpikes_ObjIcon() -> (SpireSpikes_ObjIcon:Texture 2D Object Reference)
     Option 1:Texture 2D Object Reference=Get Essence_ObjIcon() -> (Essence_ObjIcon:Texture 2D Object Reference)
     Option 2:Texture 2D Object Reference=Get Combatants_ObjIcon() -> (Combatants_ObjIcon:Texture 2D Object Reference)
     Index:Byte=0
    ) -> (ReturnValue:Texture 2D Object Reference)
    bMatchSize:bool=false
   ) -> (then)
   SetVisibility(execute=@n3.else, self=Get HBox_ObjectiveProgress(), InVisibility=Collapsed)
  ]}
 UpdateHealthInvulnerabilityState {nodes [
   n1 = UpdateHealthInvulnerabilityState!() -> (then, IsVulnerable:bool=~) {is=title K2Node_FunctionEntry}
   > SetVisibility(
    self=Get Image_LockedState() -> (Image_LockedState:Image Object Reference)
    InVisibility=Select!(Option 0:ESlateVisibility Enum=SelfHitTestInvisible, Option 1:ESlateVisibility Enum=Collapsed, Index:bool=@n1) -> (ReturnValue:ESlateVisibility Enum)
   )
   n2 = Get ProgressBar_HealthFill()
   > SetColorA!(
    execute
    self:Fort Simple Material Progress Bar Object Reference=@n2
    ColorA:Linear Color Structure=SelectColor(
     A=Get ActiveColorA() -> (ActiveColorA:Linear Color Structure)
     B=Get LockedColor() -> (LockedColor:Linear Color Structure)
     bPickA=Get IsVulnerable() -> (IsVulnerable:bool)
    )
   ) -> (then)
   > SetColorB!(
    execute
    self:Fort Simple Material Progress Bar Object Reference=@n2
    ColorB:Linear Color Structure=SelectColor(A=Get ActiveColorB() -> (ActiveColorB:Linear Color Structure), B=Get LockedColor(), bPickA=Get IsVulnerable())
   ) -> (then)
  ]}
 UpdateHealthStatusVisibility {nodes [
   n1 = UpdateHealthStatusVisibility!() -> (then, CurrentDistanceThreshold:int) {is=title K2Node_FunctionEntry}
   n2 = "Equal (Integer)"(A=@n1)
   > Branch(Condition="Not Equal (Boolean)"(A=IsVisible(), B=@n2) -> (ReturnValue=~))
   > SetVisibility(InVisibility=Select(Option 0=Collapsed, Option 1=SelfHitTestInvisible, Index=@n2))
  ]}
 UpdateObjectiveProgress {nodes [
   n1 = UpdateObjectiveProgress!() -> (then, InCurrentProgress:int) {is=title K2Node_FunctionEntry}
   > "SetText (Text)"(self=Get Text_Obj_Progress() -> (Text_Obj_Progress:Common Text Object Reference), InText="To Text (Integer)"(Value=@n1))
  ]}
]
variables [
 bIsObjectiveValid {type=boolean}
 CurrentProgress {type=number, default=1}
]
widgets {
 info {inheritedWidgetCount=1, parentClass=/Script/FortniteUI.AthenaWorldUILabelWidgetBase, rootWidgetClass=&Overlay, widgetCount=15}
 widgets [
  Overlay_Bar {widgetClassPath=&Overlay, widget=Overlay_Bar}
  VerticalBox_0 {
   widgetClassPath=&VerticalBox
   parent=Overlay_Bar
   widget=VerticalBox_0
   slot {class=OverlaySlot, properties {horizontalAlignment=HAlign_Fill, verticalAlignment=VAlign_Fill}}
  }
  HBox_ObjectiveProgress {
   widgetClassPath=&HorizontalBox
   bIsVariable=true
   parent=VerticalBox_0
   widget=HBox_ObjectiveProgress
   properties {visibility=Collapsed}
   slot {class=VerticalBoxSlot, properties {horizontalAlignment=HAlign_Center, verticalAlignment=VAlign_Top}}
  }
  Image_ObjectiveIcon {
   widgetClassPath=&Image
   bIsVariable=true
   parent=HBox_ObjectiveProgress
   widget=Image_ObjectiveIcon
   properties {brush {
     bIsDynamicallyLoaded=false
     drawAs=Image
     imageType=FullColor
     mirroring=NoMirror
     resourceName=None
     resourceObject=/SpireBoss/UI/Materials/T_UI_Icon_Minimap_MinorSpire
     tiling=NoTile
     imageSize=<32 32>
     margin=<ltrb 0 0 0 0>
     outlineSettings {
      bUseBrushTransparency=false
      roundingType=HalfHeightRadius
      width=0
      color {colorUseRule=UseColor_Specified, specifiedColor=<rgba 0 0 0 0>}
      cornerRadii=<0 0 0 0>
     }
     tintColor {colorUseRule=UseColor_Specified, specifiedColor=<rgba 1 1 1 1>}
     uVRegion {bIsValid=false, max=<0 0>, min=<0 0>}
    }}
   slot {class=HorizontalBoxSlot, properties {horizontalAlignment=HAlign_Center, verticalAlignment=VAlign_Center}}
  }
  Text_Obj_Progress {
   widgetClassPath=&CommonTextBlock
   bIsVariable=true
   parent=HBox_ObjectiveProgress
   widget=Text_Obj_Progress
   properties {
    autoWrapText=true
    justification=Center
    text=100
    textTransformPolicy=ToUpper
    font {
     bForceMonospaced=false
     bMaterialIsStencil=false
     fontMaterial=null
     fontObject=&Burbank
     letterSpacing=0
     monospacedWidth=1
     size=18
     skewAmount=0
     typefaceFontName=Burbank-Black
     outlineSettings {
      bApplyOutlineToDropShadows=false
      bMiteredCorners=false
      bSeparateFillAlpha=false
      outlineMaterial=null
      outlineSize=0
      outlineColor=<rgba 0 0 0 1>
     }
    }
   }
   slot {class=HorizontalBoxSlot, properties {padding=<ltrb 5 0 0 0>}}
  }
  Text_ {
   widgetClassPath=&CommonTextBlock
   parent=HBox_ObjectiveProgress
   widget=Text_
   properties {
    autoWrapText=true
    justification=Center
    text=/
    textTransformPolicy=ToUpper
    font {
     bForceMonospaced=false
     bMaterialIsStencil=false
     fontMaterial=null
     fontObject=&Burbank
     letterSpacing=0
     monospacedWidth=1
     size=18
     skewAmount=0
     typefaceFontName=Burbank-Black
     outlineSettings {
      bApplyOutlineToDropShadows=false
      bMiteredCorners=false
      bSeparateFillAlpha=false
      outlineMaterial=null
      outlineSize=0
      outlineColor=<rgba 0 0 0 1>
     }
    }
   }
   slot {class=HorizontalBoxSlot, properties {padding=<ltrb 5 0 0 0>}}
  }
  Text_Obj_Max {
   widgetClassPath=&CommonTextBlock
   bIsVariable=true
   parent=HBox_ObjectiveProgress
   widget=Text_Obj_Max
   properties {
    autoWrapText=true
    justification=Center
    text=100
    textTransformPolicy=ToUpper
    font {
     bForceMonospaced=false
     bMaterialIsStencil=false
     fontMaterial=null
     fontObject=&Burbank
     letterSpacing=0
     monospacedWidth=1
     size=18
     skewAmount=0
     typefaceFontName=Burbank-Black
     outlineSettings {
      bApplyOutlineToDropShadows=false
      bMiteredCorners=false
      bSeparateFillAlpha=false
      outlineMaterial=null
      outlineSize=0
      outlineColor=<rgba 0 0 0 1>
     }
    }
   }
   slot {class=HorizontalBoxSlot, properties {padding=<ltrb 5 0 0 0>}}
  }
  HorizontalBox_0 {
   widgetClassPath=&HorizontalBox
   parent=VerticalBox_0
   widget=HorizontalBox_0
   slot {class=VerticalBoxSlot, properties {horizontalAlignment=HAlign_Left, verticalAlignment=VAlign_Top, padding=<ltrb 0 5 0 0>}}
  }
  Image_LockedState {
   widgetClassPath=&Image
   bIsVariable=true
   parent=HorizontalBox_0
   widget=Image_LockedState
   properties {
    visibility=Collapsed
    brush {
     bIsDynamicallyLoaded=false
     drawAs=Image
     imageType=FullColor
     mirroring=NoMirror
     resourceName=None
     resourceObject=/Game/UI/Foundation/Textures/Icons/Locks/T-Icon-Lock-32
     tiling=NoTile
     imageSize=<18 18>
     margin=<ltrb 0 0 0 0>
     outlineSettings {
      bUseBrushTransparency=false
      roundingType=HalfHeightRadius
      width=0
      color {colorUseRule=UseColor_Specified, specifiedColor=<rgba 0 0 0 0>}
      cornerRadii=<0 0 0 0>
     }
     tintColor {colorUseRule=UseColor_Specified, specifiedColor=<rgba 1 1 1 1>}
     uVRegion {bIsValid=false, max=<0 0>, min=<0 0>}
    }
   }
   slot {class=HorizontalBoxSlot, properties {horizontalAlignment=HAlign_Center, verticalAlignment=VAlign_Center}}
  }
  SizeBox_Health {
   widgetClassPath=&SizeBox
   bIsVariable=true
   parent=HorizontalBox_0
   widget=SizeBox_Health
   properties {bOverride_HeightOverride=true, bOverride_MinDesiredWidth=true, heightOverride=12, minDesiredWidth=200}
   slot {class=HorizontalBoxSlot, properties {verticalAlignment=VAlign_Center}}
  }
  Overlay_33 {widgetClassPath=&Overlay, parent=SizeBox_Health, widget=Overlay_33, slot {class=SizeBoxSlot}}
  Image_HealthBar_BG {
   widgetClassPath=&Image
   parent=Overlay_33
   widget=Image_HealthBar_BG
   properties {brush {
     bIsDynamicallyLoaded=false
     drawAs=NoDrawType
     imageType=NoImage
     mirroring=NoMirror
     resourceName=None
     resourceObject=null
     tiling=NoTile
     imageSize=<32 32>
     margin=<ltrb 0 0 0 0>
     outlineSettings {
      bUseBrushTransparency=false
      roundingType=HalfHeightRadius
      width=0
      color {colorUseRule=UseColor_Specified, specifiedColor=<rgba 0 0 0 0>}
      cornerRadii=<0 0 0 0>
     }
     tintColor {colorUseRule=UseColor_Specified, specifiedColor=<rgba 1 1 1 1>}
     uVRegion {bIsValid=false, max=<0 0>, min=<0 0>}
    }}
   slot {class=OverlaySlot, properties {horizontalAlignment=HAlign_Fill, verticalAlignment=VAlign_Fill}}
  }
  ProgressBar_HealthDelta {
   widgetClassPath=/Game/Athena/HUD/SimpleMaterialProgressBar
   bIsVariable=true
   parent=Overlay_33
   widget=ProgressBar_HealthDelta
   properties {
    bOverrideDefaultColorA=true
    bOverrideDefaultColorB=true
    bOverrideDefaultColorBackground=true
    bOverrideDefaultSegmentEdge=true
    bOverrideDefaultSegments=true
    designTime_Progress=0.75
    segmentEdge=0.0007
    segments=1
    visibility=Collapsed
    cachedColorA=<rgba 1 0 0.0271 0>
    cachedColorB=<rgba 1 0 0 0>
    cachedSegmentColor=<rgba 0 0 0 1>
   }
   slot {class=OverlaySlot, properties {horizontalAlignment=HAlign_Fill, verticalAlignment=VAlign_Fill}}
  }
  ProgressBar_HealthFill {
   widgetClassPath=&SimpleMaterialProgressBar
   bIsVariable=true
   parent=Overlay_33
   widget=ProgressBar_HealthFill
   properties {
    bOverrideDefaultColorA=true
    bOverrideDefaultColorB=true
    bOverrideDefaultColorBackground=true
    bOverrideDefaultSegmentEdge=true
    bOverrideDefaultSegments=true
    designTime_Progress=0.5
    segmentEdge=0.0007
    segments=1
    cachedColorA=<rgba 0.8714 0.6724 0.1779 1>
    cachedColorB=<rgba 0.5906 0.3419 0.0648 1>
    cachedColorBackground=<rgba 0 0 0 0.25>
    cachedSegmentColor=<rgba 0 0 0 1>
   }
   slot {class=OverlaySlot, properties {horizontalAlignment=HAlign_Fill, verticalAlignment=VAlign_Fill}}
  }
  TutorialHighlight {widgetClassPath=&TutorialHighlight}
 ]
}
